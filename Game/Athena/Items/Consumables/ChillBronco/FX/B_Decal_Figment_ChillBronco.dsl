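; Blueprint decal actor for the ChillBronco consumable's Figment hit effect, going by the material and asset names.
; Its construction script creates a dynamic material instance from the ChillBronco hit decal material.
; It stores that instance in the Decal_MID variable and assigns it to the actor's decal component.
; No parameters are set on the instance in this script.
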(fn ConstructionScript
  (bind _returnvalue (Rendering|Material|CreateDynamicMaterialInstance "/Game/Athena/Items/Consumables/ChillBronco/FX/MDI_Figment_Decal_ChillBronco_Hit.MDI_Figment_Decal_ChillBronco_Hit" "None" "None" :!class /Script/Engine.KismetMaterialLibrary))
  (bind _output_get (Variables|Default|SetDecal_MID _returnvalue))
  (Rendering|Components|Decal|SetDecalMaterial (Variables|Decal|GetDecal) _output_get))
